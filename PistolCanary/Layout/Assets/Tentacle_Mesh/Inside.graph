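# Flat purple Default Lit material, named Inside, among the tentacle mesh assets and flagged for skeletal mesh and mesh deformer use.
# Constant ambient and diffuse purple, a grey specular color and a low shininess feed Interchange's MF_PhongToMetalRoughness function, which converts Phong values to metal-roughness PBR.
# It carries Interchange import data and uses no textures or exposed parameters.
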Material /PistolCanary/Layout/Assets/Tentacle_Mesh/Inside
graph {
 properties {assetImportData="<self>:InterchangeAssetImportData_0", bUsedWithMeshDeformer=true, bUsedWithSkeletalMesh=true, usedShadingModels=Default Lit}
 nodes [MaterialFunctionCall!(
   AmbientColor=Constant3Vector() {constant=<rgba 0.4479 0.009 0.586 1>}
   DiffuseColor=Constant3Vector() {constant=<rgba 0.4479 0.009 0.586 1>}
   SpecularColor=Constant3Vector() {constant=<rgba 0.27 0.27 0.27 1>}
   Shininess=Constant() {r=2}
  ) {materialFunction=/InterchangeAssets/Functions/MF_PhongToMetalRoughness}]
}
